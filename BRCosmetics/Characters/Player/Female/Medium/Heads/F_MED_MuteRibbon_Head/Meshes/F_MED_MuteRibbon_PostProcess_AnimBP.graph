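# Post-process Animation Blueprint for the female MuteRibbon head that adjusts eyelash bones and the head bone.
# AnimGraph chains additive offsets on eyelash bones in two passes, then replaces head rotation; a Remap Curves node sits unconnected.
# Each update, a backbling check sets backpack alpha and, when true, stores scaled head-versus-neck rotation.
# Both branches derive full-blink strength from upper-eyelid rotation relative to reference pose; curve-based blink setters are unconnected.
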AnimBlueprint /BRCosmetics/Characters/Player/Female/Medium/Heads/F_MED_MuteRibbon_Head/Meshes/F_MED_MuteRibbon_PostProcess_AnimBP
target_skeleton=/Game/Characters/Player/Female/Medium/Heads/F_MED_MuteRibbon_Head/Meshes/CustomSkeleton/Fortnite_Base_Head_MuteRibbon_Skeleton
graphs [
 AnimGraph {nodes [
   "Transform (Modify) Bone - Bone: head"(Alpha=1, Translation="0,0,0", Rotation="0,0,0", Scale="1,1,1") {
    rotationMode=BMM_Replace
    rotationSpace=BCS_ParentBoneSpace
    scaleMode=BMM_Replace
    translationSpace=BCS_BoneSpace
    alphaBoolBlend {blendOption=Linear}
    alphaScaleBiasClamp {interpSpeedDecreasing=0.5, interpSpeedIncreasing=0.5}
    boneToModify {boneName=head}
   }
   Remap Curves!(SourcePose:Pose Link Structure) -> (Pose:Pose Link Structure) {is=title AnimGraphNode_RemapCurves, expressionList {assignmentExpressions="L_blink_pose\r\n"}}
   n1 = "Transform (Modify) Bone - Bone: dyn_eyelashes_01"(ComponentPose=Local To Component(LocalPose=Input Pose()), Alpha=1, Translation="0,0,0.3", Rotation="0,0,0", Scale="1,1,1") {
    rotationMode=BMM_Additive
    rotationSpace=BCS_BoneSpace
    translationMode=BMM_Additive
    translationSpace=BCS_BoneSpace
    alphaBoolBlend {blendOption=Linear}
    alphaScaleBiasClamp {bInterpResult=true, interpSpeedDecreasing=20, interpSpeedIncreasing=20}
    boneToModify {boneName=dyn_eyelashes_01}
    translation {z=0.3}
   }
   n2 = "Transform (Modify) Bone - Bone: dyn_eyelashes_02"(ComponentPose=@n1, Alpha=1, Translation="0,0,0.2", Rotation="0,0,0", Scale="1,1,1") {
    rotationMode=BMM_Additive
    rotationSpace=BCS_BoneSpace
    translationMode=BMM_Additive
    translationSpace=BCS_BoneSpace
    alphaBoolBlend {blendOption=Linear}
    alphaScaleBiasClamp {bInterpResult=true, interpSpeedDecreasing=20, interpSpeedIncreasing=20}
    boneToModify {boneName=dyn_eyelashes_02}
    translation {z=0.2}
   }
   n3 = "Transform (Modify) Bone - Bone: dyn_eyelashes_03"(ComponentPose=@n2, Alpha=1, Translation="0,0,0.1", Rotation="0,0,0", Scale="1,1,1") {
    rotationMode=BMM_Additive
    rotationSpace=BCS_BoneSpace
    translationMode=BMM_Additive
    translationSpace=BCS_BoneSpace
    alphaBoolBlend {blendOption=Linear}
    alphaScaleBiasClamp {bInterpResult=true, interpSpeedDecreasing=20, interpSpeedIncreasing=20}
    boneToModify {boneName=dyn_eyelashes_03}
    translation {z=0.1}
   }
   n4 = "Transform (Modify) Bone - Bone: dyn_eyelashes_04"(ComponentPose=@n3, Alpha=1, Translation="0,0,0.2", Rotation="0,0,0", Scale="1,1,1") {
    rotationMode=BMM_Additive
    rotationSpace=BCS_BoneSpace
    translationMode=BMM_Additive
    translationSpace=BCS_BoneSpace
    alphaBoolBlend {blendOption=Linear}
    alphaScaleBiasClamp {bInterpResult=true, interpSpeedDecreasing=20, interpSpeedIncreasing=20}
    boneToModify {boneName=dyn_eyelashes_04}
    translation {z=0.2}
   }
   n5 = "Transform (Modify) Bone - Bone: dyn_eyelashes_05"(ComponentPose=@n4, Alpha=1, Translation="0,0,0.3", Rotation="0,0,0", Scale="1,1,1") {
    rotationMode=BMM_Additive
    rotationSpace=BCS_BoneSpace
    translationMode=BMM_Additive
    translationSpace=BCS_BoneSpace
    alphaBoolBlend {blendOption=Linear}
    alphaScaleBiasClamp {bInterpResult=true, interpSpeedDecreasing=20, interpSpeedIncreasing=20}
    boneToModify {boneName=dyn_eyelashes_05}
    translation {z=0.3}
   }
   n6 = "Transform (Modify) Bone - Bone: dyn_Eyelashes_base"(ComponentPose=@n5, Alpha=1, Translation="0,0,0.35", Rotation="-8,0,0.5", Scale="1,1,1") {
    rotationMode=BMM_Additive
    rotationSpace=BCS_BoneSpace
    translationMode=BMM_Additive
    translationSpace=BCS_BoneSpace
    alphaBoolBlend {blendOption=Linear}
    alphaScaleBiasClamp {bInterpResult=true, interpSpeedDecreasing=20, interpSpeedIncreasing=20}
    boneToModify {boneName=dyn_Eyelashes_base}
    rotation {pitch=-8, roll=0.5}
    translation {z=0.35}
   }
   n7 = "Transform (Modify) Bone - Bone: dyn_eyelashes_01"(ComponentPose=@n6, Alpha=1, Translation="0,0,0.1", Rotation="0,0,0", Scale="1,1,1") {
    rotationMode=BMM_Additive
    rotationSpace=BCS_BoneSpace
    translationMode=BMM_Additive
    translationSpace=BCS_BoneSpace
    alphaBoolBlend {blendOption=Linear}
    alphaScaleBiasClamp {bInterpResult=true, interpSpeedDecreasing=20, interpSpeedIncreasing=20}
    boneToModify {boneName=dyn_eyelashes_01}
    translation {z=0.1}
   }
   n8 = "Transform (Modify) Bone - Bone: dyn_eyelashes_02"(ComponentPose=@n7, Alpha=1, Translation="0,0,-0.1", Rotation="0,0,0", Scale="1,1,1") {
    rotationMode=BMM_Additive
    rotationSpace=BCS_BoneSpace
    translationMode=BMM_Additive
    translationSpace=BCS_BoneSpace
    alphaBoolBlend {blendOption=Linear}
    alphaScaleBiasClamp {bInterpResult=true, interpSpeedDecreasing=20, interpSpeedIncreasing=20}
    boneToModify {boneName=dyn_eyelashes_02}
    translation {z=-0.1}
   }
   n9 = "Transform (Modify) Bone - Bone: dyn_eyelashes_03"(ComponentPose=@n8, Alpha=1, Translation="0,0,0", Rotation="0,0,0", Scale="1,1,1") {
    rotationMode=BMM_Additive
    rotationSpace=BCS_BoneSpace
    translationMode=BMM_Additive
    translationSpace=BCS_BoneSpace
    alphaBoolBlend {blendOption=Linear}
    alphaScaleBiasClamp {bInterpResult=true, interpSpeedDecreasing=20, interpSpeedIncreasing=20}
    boneToModify {boneName=dyn_eyelashes_03}
   }
   n10 = "Transform (Modify) Bone - Bone: dyn_eyelashes_05"(ComponentPose=@n9, Alpha=1, Translation="0,0,0.2", Rotation="0,0,0", Scale="1,1,1") {
    rotationMode=BMM_Additive
    rotationSpace=BCS_BoneSpace
    translationMode=BMM_Additive
    translationSpace=BCS_BoneSpace
    alphaBoolBlend {blendOption=Linear}
    alphaScaleBiasClamp {bInterpResult=true, interpSpeedDecreasing=20, interpSpeedIncreasing=20}
    boneToModify {boneName=dyn_eyelashes_05}
    translation {z=0.2}
   }
   n11 = "Transform (Modify) Bone - Bone: head"(ComponentPose=@n10, Alpha=1, Translation="0,0,0", Rotation="0,0,0", Scale="1,1,1") {
    rotationMode=BMM_Replace
    rotationSpace=BCS_ParentBoneSpace
    translationSpace=BCS_BoneSpace
    alphaBoolBlend {blendOption=Linear}
    alphaScaleBiasClamp {interpSpeedDecreasing=0.5, interpSpeedIncreasing=0.5}
    boneToModify {boneName=head}
   }
   Output Pose(Result=Component To Local(ComponentPose=@n11)) {layerGroup=None}
  ]}
 EventGraph {nodes [
   Event BlueprintUpdateAnimation()
   > n1 = Branch(Condition=CheckForBackbling())
   n2 = GetOwnerSkeletalMeshComponent()
   n3 = BreakRotator(InRot=BreakTransform(InTransform=GetDeltaTransformFromRefPose(self=@n2, BoneName=head, BaseName=neck_02)).Rotation)
   n4 = Get Head Rot Percentage() -> (Head Rot Percentage:double)
   > Set Head Rot(Head Rot:Rotator=MakeRotator(
     Roll=float * float(A=@n3, B=@n4) {is=title K2Node_PromotableOperator}
     Pitch=float * float(A=@n3.Pitch, B=@n4)
     Yaw=float * float(A=@n3.Yaw, B=@n4)
    )) -> (Output_Get:Rotator)
   > n5 = Set backpack alpha(backpack alpha:double=1) -> (Output_Get:double)
   n6 = BreakTransform(InTransform=GetDeltaTransformFromRefPose(self=GetPartSkeletalMeshComponent(), BoneName=L_eye_lid_upper_mid, BaseName=faceAttach))
   Set BlinkPose_Full(execute=[@n7.then, @n5.then], BlinkPose_Full:float=MapRangeClamped(Value=BreakRotator(InRot=@n6.Rotation), InRangeA=0, InRangeB=-20)) -> (Output_Get:float)
   n7 = Set backpack alpha(execute=@n1.else, backpack alpha=0)
   n8 = MapRangeClamped(Value=GetCurveValue(CurveName=L_blink_pose), InRangeA=0.4, InRangeB=1)
   n9 = Branch(Condition="float <= float"(A=@n8, B=0.4) -> (ReturnValue=~))
   Set BlinkPose_Full(execute=@n9.else, BlinkPose_Full=float * float(A=@n8, B=1.1))
   TryGetPawnOwner()
   float * float(A=@n8, B=0.9)
   Set BlinkPose_EyelashFlatten(BlinkPose_EyelashFlatten:float=MapRangeClamped(Value=GetCurveValue(CurveName=L_blink_pose), InRangeA=0.18, InRangeB=0.2)) -> (Output_Get:float)
  ]}
]
variables [
 backpack alpha {type=number}
 BlinkPose_EyelashFlatten {type=number}
 BlinkPose_Full {type=number}
 Head Rot {type=Rotator}
 Head Rot Percentage {type=number, default=0.8}
 LidDownPose {type=number}
]
